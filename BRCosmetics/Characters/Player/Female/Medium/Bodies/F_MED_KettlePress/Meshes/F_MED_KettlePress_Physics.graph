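# Physics asset for the KettlePress medium female player body mesh, set up for left and right jacket secondary motion.
# Capsules on the spine and thighs form the body volume around two jacket chains, each a base plus three dyn bones.
# Tight swing and twist limits link each chain, whose base constrains to spine_04.
# Unlike the left, the right base constraint sets both swings and twist Free.
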PhysicsAsset /BRCosmetics/Characters/Player/Female/Medium/Bodies/F_MED_KettlePress/Meshes/F_MED_KettlePress_Physics
bodies {
 base_l_jacket [{shapeType=Capsule, length=0.505, radius=0.505, shapeName=base_l_jacket_capsule, rotation=<rot 0 -90 0>}]
 base_r_jacket [{shapeType=Capsule, length=0.505, radius=0.505, shapeName=base_r_jacket_capsule, rotation=<rot 0 -90 0>}]
 dyn_l_jacket_1 [{shapeType=Capsule, length=1, radius=2, center=<2 0 0>, rotation=<rot -90 0 0>}]
 dyn_l_jacket_2 [{shapeType=Capsule, length=1, radius=2, center=<2 0 0>, rotation=<rot -90 0 0>}]
 dyn_l_jacket_3 [{shapeType=Capsule, length=8, radius=1.5, center=<1 0 0>, rotation=<rot 0 0 90>}]
 dyn_r_jacket_1 [{shapeType=Capsule, length=1, radius=1.5, center=<2 0 0>, rotation=<rot -90 0 0>}]
 dyn_r_jacket_2 [{shapeType=Capsule, length=1, radius=2, center=<2 0 0>, rotation=<rot -90 0 0>}]
 dyn_r_jacket_3 [{shapeType=Capsule, length=8, radius=1.5, center=<1 0 0>, rotation=<rot 0 0 90>}]
 spine_02 [{shapeType=Capsule, length=10, radius=8.5, shapeName=spine_02_capsule, center=<0 0.3102 0>}]
 spine_03 [{shapeType=Capsule, length=10, radius=9, shapeName=spine_03_capsule, center=<0 0.3102 0>}]
 spine_04 [{shapeType=Capsule, length=10.3796, radius=10.3796, shapeName=spine_04_capsule, center=<0 0.3102 0>}]
 thigh_l [{shapeType=Capsule, length=12, radius=10, shapeName=thigh_l_capsule, center=<-4.4 0.9874 0>, rotation=<rot 0 -94.2994 -90>}]
 thigh_r [{shapeType=Capsule, length=12, radius=10, shapeName=thigh_l_capsule, center=<4.4 0.9874 0>, rotation=<rot 0 -94.2994 -90>}]
}
constraints [
 {bone1Name=spine_02, bone2Name=spine_01, swing1LimitDegrees=45, swing2LimitDegrees=45, twistLimitDegrees=25}
 {bone1Name=spine_03, bone2Name=spine_02, swing1LimitDegrees=45, swing2LimitDegrees=45, twistLimitDegrees=25}
 {bone1Name=spine_04, bone2Name=spine_03, swing1LimitDegrees=45, swing2LimitDegrees=45, twistLimitDegrees=25}
 {bone1Name=base_l_jacket, bone2Name=spine_04, swing1LimitDegrees=45, swing2LimitDegrees=45, twistLimitDegrees=45}
 {bone1Name=dyn_l_jacket_1, bone2Name=base_l_jacket, swing1LimitDegrees=15, swing2LimitDegrees=25, twistLimitDegrees=1}
 {bone1Name=dyn_l_jacket_2, bone2Name=dyn_l_jacket_1, swing1LimitDegrees=15, swing2LimitDegrees=25, twistLimitDegrees=1}
 {bone1Name=dyn_l_jacket_3, bone2Name=dyn_l_jacket_2, swing1LimitDegrees=15, swing2LimitDegrees=25, twistLimitDegrees=15}
 {
  bone1Name=base_r_jacket
  bone2Name=spine_04
  swing1LimitDegrees=45
  swing1Motion=Free
  swing2LimitDegrees=45
  swing2Motion=Free
  twistLimitDegrees=45
  twistMotion=Free
 }
 {bone1Name=dyn_r_jacket_1, bone2Name=base_r_jacket, swing1LimitDegrees=15, swing2LimitDegrees=25, twistLimitDegrees=1}
 {bone1Name=dyn_r_jacket_2, bone2Name=dyn_r_jacket_1, swing1LimitDegrees=15, swing2LimitDegrees=25, twistLimitDegrees=1}
 {bone1Name=dyn_r_jacket_3, bone2Name=dyn_r_jacket_2, swing1LimitDegrees=15, swing2LimitDegrees=25, twistLimitDegrees=15}
]
