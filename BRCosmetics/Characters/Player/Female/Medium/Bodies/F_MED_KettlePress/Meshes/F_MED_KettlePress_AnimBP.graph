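# Post-process style Anim Blueprint for the F_MED_KettlePress female cosmetic body, adding deformation and secondary motion for flaps, jacket, collar and skirt.
# The copied parent pose runs through a deform Control Rig and weighted flap constraints into a cache.
# Flap AnimDynamics, constraints, gravity-overridden RigidBody and skirt-helper corrections follow.
# EventGraph derives skirt alphas from thigh-to-pelvis rotation; the jacket AnimDynamics chain is disconnected from the output.
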AnimBlueprint /BRCosmetics/Characters/Player/Female/Medium/Bodies/F_MED_KettlePress/Meshes/F_MED_KettlePress_AnimBP
target_skeleton=&/F_MED_KettlePress_Skeleton
graphs [
 AnimGraph {nodes [
   n1 = Anim Dynamics - Bone: dyn_l_jacket_1(Alpha=1, ExternalForce="0,0,0") {
    angularDampingOverride=1
    angularSpringConstant=20000
    bAngularSpring=true
    bOverrideAngularDamping=true
    bUseGravityOverride=true
    gravityScale=0.35
    simulationSpace=World
    alphaBoolBlend {blendOption=Linear}
    boundBone {boneName=dyn_l_jacket_1}
    componentLinearAccScale=<1 1 1>
    componentLinearVelScale=<1 1 1>
    gravityOverride {z=-980}
    physicsBodyDefinitions [{
      boundBone {boneName=dyn_l_jacket_1}
      boxExtents=<20 30 30>
      constraintSetup {angularLimitsMax=<0 0 60>, angularLimitsMin=<0 -15 -15>, angularTarget=<1 0 0>}
      localJointOffset=<2 0 0>
     }]
   }
   n2 = Anim Dynamics - Bone: dyn_l_jacket_2(ComponentPose=@n1, Alpha=1, ExternalForce="0,0,0") {
    angularDampingOverride=1
    angularSpringConstant=20000
    bAngularSpring=true
    bOverrideAngularDamping=true
    bUseGravityOverride=true
    gravityScale=0.35
    simulationSpace=World
    alphaBoolBlend {blendOption=Linear}
    boundBone {boneName=dyn_l_jacket_2}
    componentLinearAccScale=<1 1 1>
    componentLinearVelScale=<1 1 1>
    gravityOverride {z=-980}
    physicsBodyDefinitions [{
      boundBone {boneName=dyn_l_jacket_2}
      boxExtents=<20 30 30>
      constraintSetup {angularLimitsMax=<0 0 60>, angularLimitsMin=<0 -15 -15>, angularTarget=<1 0 0>}
      localJointOffset=<2 0 0>
     }]
   }
   n3 = Anim Dynamics - Bone: dyn_l_jacket_3(ComponentPose=@n2, Alpha=1, ExternalForce="0,0,0") {
    angularDampingOverride=1
    angularSpringConstant=20000
    bAngularSpring=true
    bOverrideAngularDamping=true
    bUseGravityOverride=true
    gravityScale=0.35
    simulationSpace=World
    alphaBoolBlend {blendOption=Linear}
    boundBone {boneName=dyn_l_jacket_3}
    componentLinearAccScale=<1 1 1>
    componentLinearVelScale=<1 1 1>
    gravityOverride {z=-980}
    physicsBodyDefinitions [{
      boundBone {boneName=dyn_l_jacket_3}
      boxExtents=<20 30 30>
      constraintSetup {angularLimitsMax=<0 0 60>, angularLimitsMin=<0 -15 -15>, angularTarget=<1 0 0>}
      localJointOffset=<2 0 0>
     }]
   }
   n4 = Anim Dynamics - Bone: dyn_r_jacket_1(ComponentPose=@n3, Alpha=1, ExternalForce="0,0,0") {
    angularDampingOverride=1
    angularSpringConstant=20000
    bAngularSpring=true
    bOverrideAngularDamping=true
    bUseGravityOverride=true
    gravityScale=0.35
    simulationSpace=World
    alphaBoolBlend {blendOption=Linear}
    boundBone {boneName=dyn_r_jacket_1}
    componentLinearAccScale=<1 1 1>
    componentLinearVelScale=<1 1 1>
    gravityOverride {z=-980}
    physicsBodyDefinitions [{
      boundBone {boneName=dyn_r_jacket_1}
      boxExtents=<20 30 30>
      constraintSetup {angularLimitsMax=<0 0 15>, angularLimitsMin=<0 -15 -15>, angularTarget=<1 0 0>}
      localJointOffset=<-2 0 0>
     }]
   }
   n5 = Anim Dynamics - Bone: dyn_r_jacket_2(ComponentPose=@n4, Alpha=1, ExternalForce="0,0,0") {
    angularDampingOverride=1
    angularSpringConstant=20000
    bAngularSpring=true
    bOverrideAngularDamping=true
    bUseGravityOverride=true
    gravityScale=0.35
    simulationSpace=World
    alphaBoolBlend {blendOption=Linear}
    boundBone {boneName=dyn_r_jacket_2}
    componentLinearAccScale=<1 1 1>
    componentLinearVelScale=<1 1 1>
    gravityOverride {z=-980}
    physicsBodyDefinitions [{
      boundBone {boneName=dyn_r_jacket_2}
      boxExtents=<20 30 30>
      constraintSetup {angularLimitsMax=<0 0 15>, angularLimitsMin=<0 -15 -15>, angularTarget=<1 0 0>}
      localJointOffset=<-2 0 0>
     }]
   }
   Anim Dynamics - Bone: dyn_r_jacket_3(ComponentPose=@n5, Alpha=1, ExternalForce="0,0,0") {
    angularDampingOverride=1
    angularSpringConstant=20000
    bAngularSpring=true
    bOverrideAngularDamping=true
    bUseGravityOverride=true
    gravityScale=0.35
    simulationSpace=World
    alphaBoolBlend {blendOption=Linear}
    boundBone {boneName=dyn_r_jacket_3}
    componentLinearAccScale=<1 1 1>
    componentLinearVelScale=<1 1 1>
    gravityOverride {z=-980}
    physicsBodyDefinitions [{
      boundBone {boneName=dyn_r_jacket_3}
      boxExtents=<20 30 30>
      constraintSetup {angularLimitsMax=<0 0 15>, angularLimitsMin=<0 -15 -15>, angularTarget=<1 0 0>}
      localJointOffset=<-2 0 0>
     }]
   }
   n6 = Anim Dynamics - Bone: dyn_flap_back_l(ComponentPose=Local To Component(LocalPose=Use cached pose 'RigCache'()), Alpha=1, ExternalForce="0,0,0") {
    angularDampingOverride=0.99
    angularSpringConstant=50
    bAngularSpring=true
    bOverrideAngularDamping=true
    bUseGravityOverride=true
    bUsePlanarLimit=false
    simulationSpace=World
    alphaBoolBlend {blendOption=Linear}
    boundBone {boneName=dyn_flap_back_l}
    componentLinearAccScale=<1 1 1>
    componentLinearVelScale=<1 1 1>
    gravityOverride {z=-980}
    physicsBodyDefinitions [{
      boundBone {boneName=dyn_flap_back_l}
      boxExtents=<15 15 15>
      constraintSetup {angularLimitsMax=<0 10 10>, angularLimitsMin=<0 -10 -10>, angularTarget=<0 0 0>}
      localJointOffset=<3 0 0>
     }]
    relativeSpaceBone {boneName=head}
   }
   n7 = Anim Dynamics - Bone: dyn_flap_back_r(ComponentPose=@n6, Alpha=1, ExternalForce="0,0,0") {
    angularDampingOverride=0.99
    angularSpringConstant=50
    bAngularSpring=true
    bOverrideAngularDamping=true
    bUseGravityOverride=true
    bUsePlanarLimit=false
    simulationSpace=World
    alphaBoolBlend {blendOption=Linear}
    boundBone {boneName=dyn_flap_back_r}
    componentLinearAccScale=<1 1 1>
    componentLinearVelScale=<1 1 1>
    gravityOverride {z=-980}
    physicsBodyDefinitions [{
      boundBone {boneName=dyn_flap_back_r}
      boxExtents=<15 15 15>
      constraintSetup {angularLimitsMax=<0 10 10>, angularLimitsMin=<0 -10 -10>, angularTarget=<0 0 0>}
      localJointOffset=<3 0 0>
     }]
    relativeSpaceBone {boneName=head}
   }
   n8 = Anim Dynamics - Bone: dyn_flap_top_l(ComponentPose=@n7, Alpha=1, ExternalForce="0,0,0") {
    angularDampingOverride=0.99
    angularSpringConstant=50
    bAngularSpring=true
    bOverrideAngularDamping=true
    bUseGravityOverride=true
    bUsePlanarLimit=false
    simulationSpace=World
    alphaBoolBlend {blendOption=Linear}
    boundBone {boneName=dyn_flap_top_l}
    componentLinearAccScale=<1 1 1>
    componentLinearVelScale=<1 1 1>
    gravityOverride {z=-980}
    physicsBodyDefinitions [{
      boundBone {boneName=dyn_flap_top_l}
      boxExtents=<15 15 15>
      constraintSetup {angularLimitsMax=<0 10 10>, angularLimitsMin=<0 -10 -10>, angularTarget=<0 0 0>}
      localJointOffset=<3 0 0>
     }]
    relativeSpaceBone {boneName=head}
   }
   n9 = Anim Dynamics - Bone: dyn_flap_top_r(ComponentPose=@n8, Alpha=1, ExternalForce="0,0,0") {
    angularDampingOverride=0.99
    angularSpringConstant=50
    bAngularSpring=true
    bOverrideAngularDamping=true
    bUseGravityOverride=true
    bUsePlanarLimit=false
    simulationSpace=World
    alphaBoolBlend {blendOption=Linear}
    boundBone {boneName=dyn_flap_top_r}
    componentLinearAccScale=<1 1 1>
    componentLinearVelScale=<1 1 1>
    gravityOverride {z=-980}
    physicsBodyDefinitions [{
      boundBone {boneName=dyn_flap_top_r}
      boxExtents=<15 15 15>
      constraintSetup {angularLimitsMax=<0 10 10>, angularLimitsMin=<0 -10 -10>, angularTarget=<0 0 0>}
      localJointOffset=<3 0 0>
     }]
    relativeSpaceBone {boneName=head}
   }
   n10 = "Constraint - dyn_l_jacket_1 (P)"(ComponentPose=@n9, Alpha=1, ConstraintWeights_0=1) {
    alphaBoolBlend {blendOption=Linear}
    boneToModify {boneName=dyn_l_jacket_1}
    constraintSetup [{targetBone {boneName=spine_03}}]
    constraintWeights [1]
   }
   n11 = "Constraint - dyn_r_jacket_1 (P)"(ComponentPose=@n10, Alpha=1, ConstraintWeights_0=1) {
    alphaBoolBlend {blendOption=Linear}
    boneToModify {boneName=dyn_r_jacket_1}
    constraintSetup [{targetBone {boneName=spine_03}}]
    constraintWeights [1]
   }
   n12 = "Constraint - dyn_l_collar (P)"(ComponentPose=@n11, Alpha=0.5, ConstraintWeights_0=1) {
    alphaBoolBlend {blendOption=Linear}
    boneToModify {boneName=dyn_l_collar}
    constraintSetup [{targetBone {boneName=clavicle_l}}]
    constraintWeights [1]
   }
   n13 = "Constraint - dyn_r_collar (P)"(ComponentPose=@n12, Alpha=0.5, ConstraintWeights_0=1) {
    alphaBoolBlend {blendOption=Linear}
    boneToModify {boneName=dyn_r_collar}
    constraintSetup [{targetBone {boneName=clavicle_r}}]
    constraintWeights [1]
   }
   n14 = RigidBody!(ComponentPose=@n13, Alpha, OverrideWorldGravity:Vector=Get GravityOverride(), ExternalForce) -> (Pose) {
    bClampLinearTranslationLimitToRefPose=true
    bOverrideWorldGravity=true
    bUseExternalClothCollision=true
    overridePhysicsAsset=&/F_MED_KettlePress_Physics
    simulationSpace=BaseBoneSpace
    alphaBoolBlend {blendOption=Linear}
    baseBoneRef {boneName=pelvis}
    componentAppliedLinearAccClamp=<3000 3000 3000>
    componentLinearAccScale=<0.3 0.3 0.3>
    componentLinearVelScale=<0.3 0.3 0.3>
    simSpaceSettings {maxAngularAcceleration=1000, maxAngularVelocity=3, maxLinearAcceleration=700, maxLinearVelocity=500, velocityScaleZ=0.2, worldAlpha=0.15}
   }
   n15 = "Transform (Modify) Bone - Bone: base_flap_back_r"!(ComponentPose=@n14, bAlphaBoolEnabled:bool=True, Translation="4,4,0", Rotation="0,0,0", Scale="1,1,1") -> (Pose) {
    alphaInputType=Bool
    rotationMode=BMM_Additive
    rotationSpace=BCS_BoneSpace
    translationMode=BMM_Additive
    translationSpace=BCS_BoneSpace
    alphaBoolBlend {blendInTime=0.2, blendOption=Linear, blendOutTime=0.2}
    boneToModify {boneName=base_flap_back_r}
    translation=<4 4>
   }
   n16 = "Transform (Modify) Bone - Bone: base_flap_back_l"!(ComponentPose=@n15, bAlphaBoolEnabled:bool=True, Translation="-4,4,0", Rotation="0,0,0", Scale="1,1,1") -> (Pose) {
    alphaInputType=Bool
    rotationMode=BMM_Additive
    rotationSpace=BCS_BoneSpace
    translationMode=BMM_Additive
    translationSpace=BCS_BoneSpace
    alphaBoolBlend {blendInTime=0.2, blendOption=Linear, blendOutTime=0.2}
    boneToModify {boneName=base_flap_back_l}
    translation=<-4 4>
   }
   n17 = "Constraint - dyn_c_front_skirthelper (P)"(ComponentPose=@n16, Alpha=Get L_leg constraintAlpha() -> (L_leg constraintAlpha:double), ConstraintWeights_0=1) {
    alphaBoolBlend {blendOption=Linear}
    alphaScaleBiasClamp {scale=0.8}
    boneToModify {boneName=dyn_c_front_skirthelper}
    constraintSetup [{targetBone {boneName=thigh_l}}]
    constraintWeights [1]
   }
   n18 = "Transform (Modify) Bone - Bone: dyn_c_front_skirthelper"(ComponentPose=@n17, Alpha=Get L_leg constraintAlpha(), Translation="0,0,0", Rotation="-17.798,118.113,-120.937", Scale="1,1,1") {
    rotationMode=BMM_Additive
    rotationSpace=BCS_BoneSpace
    scaleMode=BMM_Additive
    scaleSpace=BCS_BoneSpace
    translationMode=BMM_Additive
    translationSpace=BCS_BoneSpace
    alphaBoolBlend {blendOption=Linear}
    boneToModify {boneName=dyn_c_front_skirthelper}
    rotation=<rot -17.7979 118.113 -120.9369>
   }
   n19 = "Constraint - dyn_l_back_skirthelper (P)"(ComponentPose=@n18, Alpha=Get L leg up() -> (L leg up:double), ConstraintWeights_0=1) {
    alphaBoolBlend {blendOption=Linear}
    alphaScaleBiasClamp {scale=0.8}
    boneToModify {boneName=dyn_l_back_skirthelper}
    constraintSetup [{targetBone {boneName=thigh_l}}]
    constraintWeights [1]
   }
   n20 = "Transform (Modify) Bone - Bone: dyn_l_back_skirthelper"(ComponentPose=@n19, Alpha=Get L leg up(), Translation="0,0,0", Rotation="0,0,0", Scale="1,1,1") {
    rotationMode=BMM_Additive
    rotationSpace=BCS_BoneSpace
    scaleMode=BMM_Additive
    scaleSpace=BCS_BoneSpace
    translationMode=BMM_Additive
    translationSpace=BCS_BoneSpace
    alphaBoolBlend {blendOption=Linear}
    boneToModify {boneName=dyn_l_back_skirthelper}
   }
   n21 = "Constraint - dyn_r_back_skirthelper (P)"(ComponentPose=@n20, Alpha=Get R leg up() -> (R leg up:double), ConstraintWeights_0=1) {
    alphaBoolBlend {blendOption=Linear}
    alphaScaleBiasClamp {scale=0.8}
    boneToModify {boneName=dyn_r_back_skirthelper}
    constraintSetup [{targetBone {boneName=thigh_r}}]
    constraintWeights [1]
   }
   n22 = "Transform (Modify) Bone - Bone: dyn_r_back_skirthelper"(ComponentPose=@n21, Alpha=Get R leg up(), Translation="0,0,0", Rotation="0,0,0", Scale="1,1,1") {
    rotationMode=BMM_Additive
    rotationSpace=BCS_BoneSpace
    scaleMode=BMM_Additive
    scaleSpace=BCS_BoneSpace
    translationMode=BMM_Additive
    translationSpace=BCS_BoneSpace
    alphaBoolBlend {blendOption=Linear}
    boneToModify {boneName=dyn_r_back_skirthelper}
   }
   n23 = "Constraint - dyn_c_front_skirthelper (P)"(ComponentPose=@n22, Alpha=Get R_leg constraintAlpha() -> (R_leg constraintAlpha:double), ConstraintWeights_0=1) {
    alphaBoolBlend {blendOption=Linear}
    alphaScaleBiasClamp {scale=0.8}
    boneToModify {boneName=dyn_c_front_skirthelper}
    constraintSetup [{targetBone {boneName=thigh_r}}]
    constraintWeights [1]
   }
   n24 = "Transform (Modify) Bone - Bone: dyn_c_front_skirthelper"(ComponentPose=@n23, Alpha=Get R_leg constraintAlpha(), Translation="0,0,0", Rotation="42.755,2.696,-18.776", Scale="1,1,1") {
    rotationMode=BMM_Additive
    rotationSpace=BCS_BoneSpace
    scaleMode=BMM_Additive
    scaleSpace=BCS_BoneSpace
    translationMode=BMM_Additive
    translationSpace=BCS_BoneSpace
    alphaBoolBlend {blendOption=Linear}
    boneToModify {boneName=dyn_c_front_skirthelper}
    rotation=<rot 42.755 2.696 -18.776>
   }
   n25 = "Constraint - dyn_l_back_skirthelper (P)"(ComponentPose=@n24, Alpha=Get L leg back() -> (L leg back:double), ConstraintWeights_0=1) {
    alphaBoolBlend {blendOption=Linear}
    alphaScaleBiasClamp {scale=0.8}
    boneToModify {boneName=dyn_l_back_skirthelper}
    constraintSetup [{targetBone {boneName=thigh_l}}]
    constraintWeights [1]
   }
   n26 = "Transform (Modify) Bone - Bone: dyn_l_back_skirthelper"(ComponentPose=@n25, Alpha=Get L leg back(), Translation="0,0,0", Rotation="0,0,14", Scale="1,1,1") {
    rotationMode=BMM_Additive
    rotationSpace=BCS_BoneSpace
    scaleMode=BMM_Additive
    scaleSpace=BCS_BoneSpace
    translationMode=BMM_Additive
    translationSpace=BCS_BoneSpace
    alphaBoolBlend {blendOption=Linear}
    boneToModify {boneName=dyn_l_back_skirthelper}
    rotation {roll=14}
   }
   n27 = "Constraint - dyn_r_back_skirthelper (P)"(ComponentPose=@n26, Alpha=Get R leg back() -> (R leg back:double), ConstraintWeights_0=1) {
    alphaBoolBlend {blendOption=Linear}
    alphaScaleBiasClamp {scale=0.8}
    boneToModify {boneName=dyn_r_back_skirthelper}
    constraintSetup [{targetBone {boneName=thigh_r}}]
    constraintWeights [1]
   }
   n28 = "Transform (Modify) Bone - Bone: dyn_r_back_skirthelper"(ComponentPose=@n27, Alpha=Get R leg back(), Translation="0,0,0", Rotation="0,0,14", Scale="1,1,1") {
    rotationMode=BMM_Additive
    rotationSpace=BCS_BoneSpace
    scaleMode=BMM_Additive
    scaleSpace=BCS_BoneSpace
    translationMode=BMM_Additive
    translationSpace=BCS_BoneSpace
    alphaBoolBlend {blendOption=Linear}
    boneToModify {boneName=dyn_r_back_skirthelper}
    rotation {roll=14}
   }
   n29 = "Constraint - dyn_c_back_skirthelper (PP)"!(ComponentPose=@n28, Alpha=1, ConstraintWeights_0=0.5, ConstraintWeights_1:float=0.5) -> (Pose) {
    alphaBoolBlend {blendOption=Linear}
    alphaScaleBiasClamp {interpSpeedDecreasing=20, interpSpeedIncreasing=20, scale=0.3}
    boneToModify {boneName=dyn_c_back_skirthelper}
    constraintSetup [{targetBone {boneName=dyn_l_back_skirthelper}}, {targetBone {boneName=dyn_r_back_skirthelper}}]
    constraintWeights [1,1]
   }
   n30 = "Transform (Modify) Bone - Bone: dyn_c_back_skirthelper"(ComponentPose=@n29, Alpha=1, Translation="0,0,0", Rotation="-0.000,0.000,50.000", Scale="1,1,1") {
    rotationMode=BMM_Additive
    rotationSpace=BCS_BoneSpace
    scaleMode=BMM_Additive
    scaleSpace=BCS_BoneSpace
    translationMode=BMM_Additive
    translationSpace=BCS_BoneSpace
    alphaBoolBlend {blendOption=Linear}
    boneToModify {boneName=dyn_c_back_skirthelper}
    rotation=<rot 0 0 50>
   }
   Output Pose(Result=Component To Local(ComponentPose=@n30)) {layerGroup=None}
   n31 = Control Rig(Source=Copy Pose From Mesh() {bUseAttachedParent=true}) {
    alphaBoolBlend {blendOption=Linear}
    controlRigAssetReference {blueprintRigClass=&/F_MED_KettlePress_DeformRig.F_MED_KettlePress_DeformRig_C}
   }
   n32 = "Constraint - base_flap_front_r (PPPPPPPP)"!(
    ComponentPose=Local To Component(LocalPose=@n31)
    Alpha=1
    ConstraintWeights_0=0.4702
    ConstraintWeights_1:float=0.177
    ConstraintWeights_2:float=0.1362
    ConstraintWeights_3:float=0.0827
    ConstraintWeights_4:float=0.0794
    ConstraintWeights_5:float=0.0506
    ConstraintWeights_6:float=0.0031
    ConstraintWeights_7:float=0.0008
   ) -> (Pose) {
    alphaBoolBlend {blendOption=Linear}
    boneToModify {boneName=base_flap_front_r}
    constraintSetup [
     {targetBone {boneName=upperarm_twist_01_r}}
     {targetBone {boneName=deform_upperarm_fr_r}}
     {targetBone {boneName=clavicle_r}}
     {targetBone {boneName=spine_04}}
     {targetBone {boneName=deform_upperarm_up_r}}
     {targetBone {boneName=spine_05}}
     {targetBone {boneName=upperarm_r}}
     {targetBone {boneName=spine_02}}
    ]
    constraintWeights [1,1,1,1,1,1,1,1]
   }
   n33 = "Constraint - base_flap_top_r (PPPPPPPP)"!(
    ComponentPose=@n32
    Alpha=1
    ConstraintWeights_0=0.3605
    ConstraintWeights_1:float=0.2659
    ConstraintWeights_2:float=0.1595
    ConstraintWeights_3:float=0.1137
    ConstraintWeights_4:float=0.0465
    ConstraintWeights_5:float=0.0377
    ConstraintWeights_6:float=0.0099
    ConstraintWeights_7:float=0.0064
   ) -> (Pose) {
    alphaBoolBlend {blendOption=Linear}
    boneToModify {boneName=base_flap_top_r}
    constraintSetup [
     {targetBone {boneName=deform_upperarm_up_r}}
     {targetBone {boneName=upperarm_twist_01_r}}
     {targetBone {boneName=dyn_flap_top_r}}
     {targetBone {boneName=clavicle_r}}
     {targetBone {boneName=deform_upperarm_fr_r}}
     {targetBone {boneName=deform_upperarm_bk_r}}
     {targetBone {boneName=dyn_flap_back_r}}
     {targetBone {boneName=spine_05}}
    ]
    constraintWeights [1,1,1,1,1,1,1,1]
   }
   n34 = "Constraint - base_flap_back_r (PPPPPPPP)"!(
    ComponentPose=@n33
    Alpha=1
    ConstraintWeights_0=0.3357
    ConstraintWeights_1:float=0.247
    ConstraintWeights_2:float=0.0945
    ConstraintWeights_3:float=0.0801
    ConstraintWeights_4:float=0.0669
    ConstraintWeights_5:float=0.066
    ConstraintWeights_6:float=0.0573
    ConstraintWeights_7:float=0.0524
   ) -> (Pose) {
    alphaBoolBlend {blendOption=Linear}
    boneToModify {boneName=base_flap_back_r}
    constraintSetup [
     {targetBone {boneName=dyn_flap_back_r}}
     {targetBone {boneName=upperarm_twist_01_r}}
     {targetBone {boneName=deform_upperarm_up_r}}
     {targetBone {boneName=spine_04}}
     {targetBone {boneName=deform_upperarm_bk_r}}
     {targetBone {boneName=clavicle_r}}
     {targetBone {boneName=spine_05}}
     {targetBone {boneName=dyn_flap_top_r}}
    ]
    constraintWeights [1,1,1,1,1,1,1,1]
   }
   n35 = "Constraint - base_flap_back_l (PPPPPPPP)"!(
    ComponentPose=@n34
    Alpha=1
    ConstraintWeights_0=0.2969
    ConstraintWeights_1:float=0.2899
    ConstraintWeights_2:float=0.1269
    ConstraintWeights_3:float=0.0848
    ConstraintWeights_4:float=0.059
    ConstraintWeights_5:float=0.0558
    ConstraintWeights_6:float=0.0505
    ConstraintWeights_7:float=0.0362
   ) -> (Pose) {
    alphaBoolBlend {blendOption=Linear}
    boneToModify {boneName=base_flap_back_l}
    constraintSetup [
     {targetBone {boneName=dyn_flap_back_l}}
     {targetBone {boneName=upperarm_twist_01_l}}
     {targetBone {boneName=deform_upperarm_bk_l}}
     {targetBone {boneName=deform_upperarm_up_l}}
     {targetBone {boneName=spine_04}}
     {targetBone {boneName=clavicle_l}}
     {targetBone {boneName=spine_05}}
     {targetBone {boneName=upperarm_l}}
    ]
    constraintWeights [1,1,1,1,1,1,1,1]
   }
   n36 = "Constraint - base_flap_front_l (PPPPPPPP)"!(
    ComponentPose=@n35
    Alpha=1
    ConstraintWeights_0=0.4688
    ConstraintWeights_1:float=0.1774
    ConstraintWeights_2:float=0.136
    ConstraintWeights_3:float=0.0836
    ConstraintWeights_4:float=0.0794
    ConstraintWeights_5:float=0.0509
    ConstraintWeights_6:float=0.0032
    ConstraintWeights_7:float=0.0007
   ) -> (Pose) {
    alphaBoolBlend {blendOption=Linear}
    boneToModify {boneName=base_flap_front_l}
    constraintSetup [
     {targetBone {boneName=upperarm_twist_01_l}}
     {targetBone {boneName=deform_upperarm_fr_l}}
     {targetBone {boneName=jnt_helper_armpit_l}}
     {targetBone {boneName=spine_04}}
     {targetBone {boneName=deform_upperarm_up_l}}
     {targetBone {boneName=spine_05}}
     {targetBone {boneName=upperarm_l}}
     {targetBone {boneName=spine_02}}
    ]
    constraintWeights [1,1,1,1,1,1,1,1]
   }
   n37 = "Constraint - base_flap_top_l (PPPPPPPP)"!(
    ComponentPose=@n36
    Alpha=1
    ConstraintWeights_0=0.4902
    ConstraintWeights_1:float=0.2865
    ConstraintWeights_2:float=0.1315
    ConstraintWeights_3:float=0.0451
    ConstraintWeights_4:float=0.0408
    ConstraintWeights_5:float=0.0032
    ConstraintWeights_6:float=0.0024
    ConstraintWeights_7:float=0.0003
   ) -> (Pose) {
    alphaBoolBlend {blendOption=Linear}
    boneToModify {boneName=base_flap_top_l}
    constraintSetup [
     {targetBone {boneName=deform_upperarm_up_l}}
     {targetBone {boneName=upperarm_twist_01_l}}
     {targetBone {boneName=clavicle_l}}
     {targetBone {boneName=deform_upperarm_fr_l}}
     {targetBone {boneName=deform_upperarm_bk_l}}
     {targetBone {boneName=spine_05}}
     {targetBone {boneName=dyn_flap_top_l}}
     {targetBone {boneName=spine_04}}
    ]
    constraintWeights [1,1,1,1,1,1,1,1]
   }
   Save cached pose 'RigCache'(Pose=Component To Local(ComponentPose=@n37)) {cacheName=RigCache}
  ]}
 EventGraph {nodes [
   Event BlueprintUpdateAnimation()
   > n1 = Sequence()
   > n2 = Gravity Override State Function DEPRECATED(CharacterPartAnimInstance=Self-Reference(), GameTimeInSeconds=GetGameTimeInSeconds(), GravityOverrideParams=Get Gravity Override Params())
   > Set GravityOverride(GravityOverride=@n2)
   n3 = GetOwnerSkeletalMeshComponent()
   n4 = BreakRotator(InRot=BreakTransform(InTransform=GetDeltaTransformFromRefPose(self=@n3, BoneName=thigh_l, BaseName=pelvis)).Rotation)
   > Set L_leg constraintAlpha(L_leg constraintAlpha:double="Max (Float)"(A=MapRangeClamped(Value=@n4.Yaw, InRangeA=-30, InRangeB=-60), B=MapRangeClamped(Value=@n4.Pitch, InRangeA=5, InRangeB=-10))) -> (Output_Get:double)
   > Set L leg back(L leg back:double=MapRangeClamped(Value=@n4.Yaw, InRangeA=-30, InRangeB=-8)) -> (Output_Get:double)
   n5 = BreakRotator(InRot=BreakTransform(InTransform=GetDeltaTransformFromRefPose(self=@n3, BoneName=thigh_r, BaseName=pelvis)).Rotation)
   > Set R_leg constraintAlpha(R_leg constraintAlpha:double="Max (Float)"(A=MapRangeClamped(Value=@n5.Yaw, InRangeA=-15, InRangeB=-60), B=MapRangeClamped(Value=@n5.Pitch, InRangeA=5, InRangeB=-10))) -> (Output_Get:double)
   > Set R leg back(R leg back:double=MapRangeClamped(Value=@n5.Yaw, InRangeA=-30, InRangeB=-8)) -> (Output_Get:double)
   > Set eitherLegConstraintAlpha(eitherLegConstraintAlpha:double="Max (Float)"(A=Get L_leg constraintAlpha(), B=Get R_leg constraintAlpha())) -> (Output_Get:double)
   n6 = BreakTransform(InTransform=GetDeltaTransformFromRefPose(self=@n3, BoneName=thigh_l, BaseName=pelvis))
   > Set L leg up(L leg up:double=MapRangeClamped(Value=BreakRotator(InRot=@n6.Rotation).Yaw, InRangeA=-70, InRangeB=-110)) -> (Output_Get:double)
   n7 = BreakTransform(InTransform=GetDeltaTransformFromRefPose(self=@n3, BoneName=thigh_r, BaseName=pelvis))
   > Set R leg up(R leg up:double=MapRangeClamped(Value=BreakRotator(InRot=@n7.Rotation).Yaw, InRangeA=-70, InRangeB=-110)) -> (Output_Get:double)
   > Set eitherLegConstraintAlpha(eitherLegConstraintAlpha="Max (Float)"(A=Get L leg up(), B=Get R leg up()))
   Set turn off when crouching(execute=@n1.then_1, turn off when crouching:double="To Float (Boolean)"(InBool=Get bIsCrouching())) -> (Output_Get:double)
   > Set thigh_l_rot(thigh_l_rot:double=GetAlphaFromReferenceTransformDelta(SkeletalMeshComponent=Get SkelMesh(), BoneName=pelvis, BaseName=thigh_l, InputRange=(X=0,Y=120), TwistAxis="0,0,1")) -> (Output_Get:double)
   n8 = Event OnCharacterCustomizationLoaded()
   > Set SkelMesh(SkelMesh=@n8.CustomBaseMesh)
  ]}
]
variables [
 eitherLegConstraintAlpha {type=number}
 Gravity Override Params {type=GravityOverrideParamsStruct}
 GravityOverride {type=Vector}
 high velocity turn off anim dyn {type=number}
 L leg back {type=number}
 L leg up {type=number}
 L_leg constraintAlpha {type=number}
 R leg back {type=number}
 R leg up {type=number}
 R_leg constraintAlpha {type=number}
 SkelMesh {type=SkeletalMeshComponent}
 thigh_l_rot {type=number}
 turn off when crouching {type=number}
]
