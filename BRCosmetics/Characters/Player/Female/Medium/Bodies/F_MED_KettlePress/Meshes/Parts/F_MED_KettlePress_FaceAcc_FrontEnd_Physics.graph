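# Physics asset for the front-end variant of the female KettlePress face accessory mesh, setting up simulated hair motion.
# Four three-link capsule chains (back, left, right, right side) hang from zero-size base bodies on the head, with limited swing and mostly minimal twist.
# Head, neck, upper spine, clavicle and upper-arm capsules provide body shapes around the hair.
# Free-rotation cross-constraints link the right-side chain to the right and back chains.
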PhysicsAsset /BRCosmetics/Characters/Player/Female/Medium/Bodies/F_MED_KettlePress/Meshes/Parts/F_MED_KettlePress_FaceAcc_FrontEnd_Physics
bodies {
 base_back_hair [{shapeType=Capsule, length=0, radius=0, shapeName=base_back_hair_capsule}]
 base_l_hair [{shapeType=Capsule, length=0, radius=0, shapeName=base_l_hair_capsule}]
 base_r_hair [{shapeType=Capsule, length=0, radius=0, shapeName=base_r_hair_capsule}]
 base_r_side [{shapeType=Capsule, length=0, radius=0, shapeName=base_r_side_capsule}]
 clavicle_l [{shapeType=Capsule, length=15, radius=8, shapeName=clavicle_l_capsule, center=<13.2525 0 -2.3132>, rotation=<rot -90 180 180>}]
 clavicle_r [{shapeType=Capsule, length=7, radius=8, shapeName=clavicle_r_capsule, center=<-5.5 0 0>, rotation=<rot 90 0 0>}]
 dyn_back_hair_1 [{shapeType=Capsule, length=3, radius=3, shapeName=dyn_back_hair_1_capsule, center=<3 0 0>, rotation=<rot -90 0 0>}]
 dyn_back_hair_2 [{shapeType=Capsule, length=2, radius=2, shapeName=dyn_back_hair_2_capsule, center=<3 0 1.75>, rotation=<rot -90 180 180>}]
 dyn_back_hair_3 [{shapeType=Capsule, length=3, radius=3, shapeName=dyn_back_hair_3_capsule, center=<5 0 0>, rotation=<rot -90 0 0>}]
 dyn_l_hair_1 [{shapeType=Capsule, length=3, radius=3.5, shapeName=dyn_l_hair_1_capsule, center=<3 0 0>, rotation=<rot -90 0 0>}]
 dyn_l_hair_2 [{shapeType=Capsule, length=3, radius=2.5, shapeName=dyn_l_hair_2_capsule, center=<3 0 0>, rotation=<rot -90 0 0>}]
 dyn_l_hair_3 [{shapeType=Capsule, length=1, radius=2, shapeName=dyn_l_hair_3_capsule, center=<1 0 0.7099>, rotation=<rot -90 0 0>}]
 dyn_r_hair__1 [{shapeType=Capsule, length=3, radius=3, shapeName=dyn_r_hair__1_capsule, center=<3 0 0>, rotation=<rot -90 0 0>}]
 dyn_r_hair__2 [{shapeType=Capsule, length=3, radius=3, shapeName=dyn_r_hair__2_capsule, center=<3 0 0>, rotation=<rot -90 0 0>}]
 dyn_r_hair__3 [{shapeType=Capsule, length=1, radius=3, shapeName=dyn_r_hair__3_capsule, center=<1 0 0>, rotation=<rot -90 0 0>}]
 dyn_r_side_1 [{shapeType=Capsule, length=3, radius=3.5, shapeName=dyn_r_side_1_capsule, center=<3 0 0>, rotation=<rot -90 0 0>}]
 dyn_r_side_2 [{shapeType=Capsule, length=3, radius=3.5, shapeName=dyn_r_side_2_capsule, center=<3 0 0>, rotation=<rot -90 180 180>}]
 dyn_r_side_3 [{shapeType=Capsule, length=6, radius=3, shapeName=dyn_r_side_2_capsule, center=<1 0 0>, rotation=<rot 0 10 80>}]
 head [
  {shapeType=Capsule, length=4, radius=9, shapeName=head_capsule, center=<2.6435 -1 0>, rotation=<rot -90 180 180>}
  {shapeType=Capsule, length=25, radius=3, shapeName=head_capsule, center=<2.2726 -0.5026 -9.2288>}
 ]
 neck_02 [{shapeType=Capsule, length=4, radius=9, shapeName=head_capsule, center=<2.6435 -2.3975 0>, rotation=<rot -90 0 0>}]
 spine_04 [{shapeType=Capsule, length=20, radius=13, shapeName=spine_05_capsule, center=<0 -1.3189 0>, rotation=<rot 0 -90 0>}]
 spine_05 [{shapeType=Capsule, length=20, radius=12, shapeName=spine_05_capsule, center=<-1.1233 -0.7395 0>, rotation=<rot 0 -90 0>}]
 upperarm_l [{shapeType=Capsule, length=30, radius=6.1463, shapeName=upperarm_l_capsule, center=<15 0 0>, rotation=<rot -90 0 0>}]
 upperarm_r [{shapeType=Capsule, length=30, radius=7, shapeName=upperarm_r_capsule, center=<-17.1393 0 0>, rotation=<rot -90 0 0>}]
}
constraints [
 {bone1Name=clavicle_l, bone2Name=spine_05, swing1LimitDegrees=45, swing2LimitDegrees=45, twistLimitDegrees=25}
 {bone1Name=clavicle_r, bone2Name=spine_05, swing1LimitDegrees=45, swing2LimitDegrees=45, twistLimitDegrees=25}
 {bone1Name=head, bone2Name=neck_02, swing1LimitDegrees=45, swing2LimitDegrees=45, twistLimitDegrees=25}
 {
  bone1Name=upperarm_l
  bone2Name=clavicle_l
  swing1LimitDegrees=45
  swing1Motion=Free
  swing2LimitDegrees=45
  swing2Motion=Free
  twistLimitDegrees=45
  twistMotion=Free
 }
 {
  bone1Name=upperarm_r
  bone2Name=clavicle_r
  swing1LimitDegrees=45
  swing1Motion=Free
  swing2LimitDegrees=45
  swing2Motion=Free
  twistLimitDegrees=45
  twistMotion=Free
 }
 {
  bone1Name=neck_02
  bone2Name=spine_05
  swing1LimitDegrees=45
  swing1Motion=Free
  swing2LimitDegrees=45
  swing2Motion=Free
  twistLimitDegrees=45
  twistMotion=Free
 }
 {bone1Name=dyn_r_side_3, bone2Name=dyn_r_side_2, swing1LimitDegrees=10, swing2LimitDegrees=30, twistLimitDegrees=1}
 {
  bone1Name=base_back_hair
  bone2Name=head
  swing1LimitDegrees=45
  swing1Motion=Free
  swing2LimitDegrees=45
  swing2Motion=Free
  twistLimitDegrees=45
  twistMotion=Free
 }
 {bone1Name=dyn_back_hair_1, bone2Name=base_back_hair, swing1LimitDegrees=10, swing2LimitDegrees=10, twistLimitDegrees=1}
 {bone1Name=dyn_back_hair_2, bone2Name=dyn_back_hair_1, swing1LimitDegrees=10, swing2LimitDegrees=10, twistLimitDegrees=1}
 {bone1Name=dyn_back_hair_3, bone2Name=dyn_back_hair_2, swing1LimitDegrees=10, swing2LimitDegrees=10, twistLimitDegrees=1}
 {
  bone1Name=base_l_hair
  bone2Name=head
  swing1LimitDegrees=45
  swing1Motion=Free
  swing2LimitDegrees=45
  swing2Motion=Free
  twistLimitDegrees=45
  twistMotion=Free
 }
 {bone1Name=dyn_l_hair_1, bone2Name=base_l_hair, swing1LimitDegrees=30, swing2LimitDegrees=30, twistLimitDegrees=15}
 {bone1Name=dyn_l_hair_2, bone2Name=dyn_l_hair_1, swing1LimitDegrees=30, swing2LimitDegrees=30, twistLimitDegrees=1}
 {bone1Name=dyn_l_hair_3, bone2Name=dyn_l_hair_2, swing1LimitDegrees=30, swing2LimitDegrees=30, twistLimitDegrees=1}
 {
  bone1Name=base_r_hair
  bone2Name=head
  swing1LimitDegrees=45
  swing1Motion=Free
  swing2LimitDegrees=45
  swing2Motion=Free
  twistLimitDegrees=45
  twistMotion=Free
 }
 {bone1Name=dyn_r_hair__1, bone2Name=base_r_hair, swing1LimitDegrees=20, swing2LimitDegrees=30, twistLimitDegrees=1}
 {bone1Name=dyn_r_hair__2, bone2Name=dyn_r_hair__1, swing1LimitDegrees=20, swing2LimitDegrees=30, twistLimitDegrees=1}
 {bone1Name=dyn_r_hair__3, bone2Name=dyn_r_hair__2, swing1LimitDegrees=20, swing2LimitDegrees=30, twistLimitDegrees=1}
 {
  bone1Name=base_r_side
  bone2Name=head
  swing1LimitDegrees=45
  swing1Motion=Free
  swing2LimitDegrees=45
  swing2Motion=Free
  twistLimitDegrees=45
  twistMotion=Free
 }
 {bone1Name=dyn_r_side_1, bone2Name=base_r_side, swing1LimitDegrees=10, swing2LimitDegrees=20, twistLimitDegrees=1}
 {bone1Name=dyn_r_side_2, bone2Name=dyn_r_side_1, swing1LimitDegrees=10, swing2LimitDegrees=30, twistLimitDegrees=1}
 {
  bone1Name=spine_05
  bone2Name=spine_04
  swing1LimitDegrees=45
  swing1Motion=Free
  swing2LimitDegrees=45
  swing2Motion=Free
  twistLimitDegrees=45
  twistMotion=Free
 }
 {
  bone1Name=dyn_r_hair__3
  bone2Name=dyn_r_side_3
  swing1LimitDegrees=45
  swing1Motion=Free
  swing2LimitDegrees=45
  swing2Motion=Free
  twistLimitDegrees=45
  twistMotion=Free
 }
 {
  bone1Name=dyn_r_hair__2
  bone2Name=dyn_r_side_2
  swing1LimitDegrees=45
  swing1Motion=Free
  swing2LimitDegrees=45
  swing2Motion=Free
  twistLimitDegrees=45
  twistMotion=Free
 }
 {
  bone1Name=dyn_back_hair_3
  bone2Name=dyn_r_side_3
  swing1LimitDegrees=45
  swing1Motion=Free
  swing2LimitDegrees=45
  swing2Motion=Free
  twistLimitDegrees=45
  twistMotion=Free
 }
]
