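# Physics asset for the KettlePress female face accessory's hair during skydiving.
# Four three-segment capsule chains (back, left, right, right side) hang from base bones constrained to the head, with capsules on head, neck, upper spine, clavicles and upper arms.
# Chain joints have tight angular limits, and left and right hair lock one swing axis.
# An extra free-angular constraint links the right side tip to the right hair tip.
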PhysicsAsset /BRCosmetics/Characters/Player/Female/Medium/Bodies/F_MED_KettlePress/Meshes/Parts/F_MED_KettlePress_FaceAcc_Skydiving_Physics
bodies {
 base_back_hair [{shapeType=Capsule, length=1, radius=1, rotation=<rot -90 0 0>}]
 base_l_hair [{shapeType=Capsule, length=1, radius=1, rotation=<rot -90 0 0>}]
 base_r_hair [{shapeType=Capsule, length=1, radius=1, rotation=<rot -90 0 0>}]
 base_r_side [{shapeType=Capsule, length=1, radius=1, rotation=<rot -90 0 0>}]
 clavicle_l [{shapeType=Capsule, length=6.1463, radius=7.5, shapeName=clavicle_l_capsule, center=<5.5 0 0>, rotation=<rot -90 0 0>}]
 clavicle_r [{shapeType=Capsule, length=6.1463, radius=7.5, shapeName=clavicle_r_capsule, center=<-5.5 0 0>, rotation=<rot -90 0 0>}]
 dyn_back_hair_1 [{shapeType=Capsule, length=3, radius=2, center=<3 0 1.2453>, rotation=<rot -90 0 0>}]
 dyn_back_hair_2 [{shapeType=Capsule, length=3, radius=4.5, center=<3 0 0>, rotation=<rot -90 0 0>}]
 dyn_back_hair_3 [{shapeType=Capsule, length=3, radius=2, center=<3 0 1.2453>, rotation=<rot -90 0 0>}]
 dyn_l_hair_1 [{shapeType=Capsule, length=3, radius=3, center=<3 0 0>, rotation=<rot -90 0 0>}]
 dyn_l_hair_2 [{shapeType=Capsule, length=3, radius=3, center=<3 0 -0.7532>, rotation=<rot -90 180 180>}]
 dyn_l_hair_3 [{shapeType=Capsule, length=8, radius=4, center=<3.8074 0 -0.8243>, rotation=<rot 0 0 90>}]
 dyn_r_hair__1 [{shapeType=Capsule, length=3, radius=3, center=<3 0 0>, rotation=<rot -90 0 0>}]
 dyn_r_hair__2 [{shapeType=Capsule, length=3, radius=3, center=<3 0 0>, rotation=<rot -90 0 0>}]
 dyn_r_hair__3 [{shapeType=Capsule, length=3, radius=3, center=<3 0 0>, rotation=<rot -90 0 0>}]
 dyn_r_side_1 [{shapeType=Capsule, length=3, radius=3, center=<3 0 0>, rotation=<rot -90 0 0>}]
 dyn_r_side_2 [{shapeType=Capsule, length=3, radius=3, center=<3 0 0>, rotation=<rot -90 0 0>}]
 dyn_r_side_3 [{shapeType=Capsule, length=3, radius=3, center=<3 0 0>, rotation=<rot -90 0 0>}]
 head [
  {shapeType=Capsule, length=8.8258, radius=8.8258, shapeName=head_capsule, center=<2.6435 0 0>, rotation=<rot -90 0 0>}
  {shapeType=Capsule, length=25, radius=3.5, shapeName=head_capsule, center=<0.0196 -0.1226 -9.2288>}
 ]
 neck_02 [
  {shapeType=Capsule, length=4, radius=9, shapeName=head_capsule, center=<2.6435 -2.3975 0>, rotation=<rot -90 0 0>}
  {shapeType=Capsule, length=40, radius=15, shapeName=neck_02_capsule, center=<-5.2601 24.5954 0>, rotation=<rot 0 -90 0>}
 ]
 spine_05 [{shapeType=Capsule, length=20, radius=13, shapeName=spine_05_capsule, center=<0 0.0783 0>, rotation=<rot 0 -90 0>}]
 upperarm_l [{shapeType=Capsule, length=30, radius=8, shapeName=upperarm_l_capsule, center=<15 0 0>, rotation=<rot -90 0 0>}]
 upperarm_r [{shapeType=Capsule, length=30, radius=6.1463, shapeName=upperarm_r_capsule, center=<-17.1393 0 0>, rotation=<rot -90 0 0>}]
}
constraints [
 {bone1Name=clavicle_l, bone2Name=spine_05, swing1LimitDegrees=45, swing2LimitDegrees=45, twistLimitDegrees=25}
 {bone1Name=clavicle_r, bone2Name=spine_05, swing1LimitDegrees=45, swing2LimitDegrees=45, twistLimitDegrees=25}
 {bone1Name=head, bone2Name=neck_02, swing1LimitDegrees=45, swing2LimitDegrees=45, twistLimitDegrees=25}
 {
  bone1Name=upperarm_l
  bone2Name=clavicle_l
  swing1LimitDegrees=45
  swing1Motion=Free
  swing2LimitDegrees=45
  swing2Motion=Free
  twistLimitDegrees=45
  twistMotion=Free
 }
 {
  bone1Name=upperarm_r
  bone2Name=clavicle_r
  swing1LimitDegrees=45
  swing1Motion=Free
  swing2LimitDegrees=45
  swing2Motion=Free
  twistLimitDegrees=45
  twistMotion=Free
 }
 {
  bone1Name=neck_02
  bone2Name=spine_05
  swing1LimitDegrees=45
  swing1Motion=Free
  swing2LimitDegrees=45
  swing2Motion=Free
  twistLimitDegrees=45
  twistMotion=Free
 }
 {bone1Name=dyn_r_side_3, bone2Name=dyn_r_side_2, swing1LimitDegrees=1, swing2LimitDegrees=15, twistLimitDegrees=1}
 {
  bone1Name=base_back_hair
  bone2Name=head
  swing1LimitDegrees=45
  swing1Motion=Free
  swing2LimitDegrees=45
  swing2Motion=Free
  twistLimitDegrees=45
  twistMotion=Free
 }
 {bone1Name=dyn_back_hair_1, bone2Name=base_back_hair, swing1LimitDegrees=2, swing2LimitDegrees=25, twistLimitDegrees=1}
 {bone1Name=dyn_back_hair_2, bone2Name=dyn_back_hair_1, swing1LimitDegrees=2, swing2LimitDegrees=25, twistLimitDegrees=1}
 {bone1Name=dyn_back_hair_3, bone2Name=dyn_back_hair_2, swing1LimitDegrees=2, swing2LimitDegrees=25, twistLimitDegrees=1}
 {
  bone1Name=base_l_hair
  bone2Name=head
  swing1LimitDegrees=45
  swing1Motion=Free
  swing2LimitDegrees=45
  swing2Motion=Free
  twistLimitDegrees=45
  twistMotion=Free
 }
 {bone1Name=dyn_l_hair_1, bone2Name=base_l_hair, swing1LimitDegrees=10, swing1Motion=Locked, swing2LimitDegrees=10, twistLimitDegrees=1}
 {bone1Name=dyn_l_hair_2, bone2Name=dyn_l_hair_1, swing1LimitDegrees=10, swing1Motion=Locked, swing2LimitDegrees=10, twistLimitDegrees=1}
 {bone1Name=dyn_l_hair_3, bone2Name=dyn_l_hair_2, swing1LimitDegrees=10, swing1Motion=Locked, swing2LimitDegrees=10, twistLimitDegrees=1}
 {
  bone1Name=base_r_hair
  bone2Name=head
  swing1LimitDegrees=45
  swing1Motion=Free
  swing2LimitDegrees=45
  swing2Motion=Free
  twistLimitDegrees=45
  twistMotion=Free
 }
 {bone1Name=dyn_r_hair__1, bone2Name=base_r_hair, swing1LimitDegrees=20, swing1Motion=Locked, swing2LimitDegrees=5, twistLimitDegrees=1}
 {bone1Name=dyn_r_hair__2, bone2Name=dyn_r_hair__1, swing1LimitDegrees=20, swing1Motion=Locked, swing2LimitDegrees=10, twistLimitDegrees=1}
 {bone1Name=dyn_r_hair__3, bone2Name=dyn_r_hair__2, swing1LimitDegrees=20, swing1Motion=Locked, swing2LimitDegrees=10, twistLimitDegrees=1}
 {
  bone1Name=base_r_side
  bone2Name=head
  swing1LimitDegrees=45
  swing1Motion=Free
  swing2LimitDegrees=45
  swing2Motion=Free
  twistLimitDegrees=45
  twistMotion=Free
 }
 {bone1Name=dyn_r_side_1, bone2Name=base_r_side, swing1LimitDegrees=1, swing2LimitDegrees=15, twistLimitDegrees=1}
 {bone1Name=dyn_r_side_2, bone2Name=dyn_r_side_1, swing1LimitDegrees=1, swing2LimitDegrees=15, twistLimitDegrees=1}
 {
  bone1Name=dyn_r_side_3
  bone2Name=dyn_r_hair__3
  swing1LimitDegrees=45
  swing1Motion=Free
  swing2LimitDegrees=45
  swing2Motion=Free
  twistLimitDegrees=45
  twistMotion=Free
 }
]
