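# Animation Blueprint for the KettlePress female character's face accessory mesh, adding secondary motion to hair and earrings.
# Copied parent and body poses feed RigidBody simulations for front end, skydiving or normal play, with bone constraints replacing them when rigid bodies are disabled.
# Additive hair offsets precede simulation, driven partly by head tilt and crouch or parachute events.
# AnimDynamics swings both earrings last in every branch.
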AnimBlueprint /BRCosmetics/Characters/Player/Female/Medium/Bodies/F_MED_KettlePress/Meshes/Parts/F_MED_KettlePress_FaceAcc_AnimBP
target_skeleton=&/F_MED_KettlePress_FaceAcc_Skeleton
graphs [
 AnimGraph {nodes [
   n1 = Get head up() -> (head up:double)
   n2 = "Transform (Modify) Bone - Bone: base_r_side"(
    ComponentPose=Local To Component(LocalPose=Use cached pose 'StartPose'())
    Alpha=@n1
    Translation="0,0,0"
    Rotation=Get RotationHair_SideR() -> (RotationHair_SideR:Rotator)
    Scale="1,1,1"
   ) {
    rotationMode=BMM_Additive
    rotationSpace=BCS_BoneSpace
    translationMode=BMM_Additive
    translationSpace=BCS_BoneSpace
    alphaBoolBlend {blendOption=Linear}
    boneToModify {boneName=base_r_side}
   }
   n3 = "Transform (Modify) Bone - Bone: base_back_hair"(ComponentPose=@n2, Alpha=@n1, Translation="0,0,0", Rotation=Get RotationHair_Back() -> (RotationHair_Back:Rotator), Scale="1,1,1") {
    rotationMode=BMM_Additive
    rotationSpace=BCS_BoneSpace
    translationMode=BMM_Additive
    translationSpace=BCS_BoneSpace
    alphaBoolBlend {blendOption=Linear}
    boneToModify {boneName=base_back_hair}
   }
   n4 = "Transform (Modify) Bone - Bone: base_r_hair"(ComponentPose=@n3, Alpha=@n1, Translation="0,0,0", Rotation=Get RotationHair_Right() -> (RotationHair_Right:Rotator), Scale="1,1,1") {
    rotationMode=BMM_Additive
    rotationSpace=BCS_BoneSpace
    translationMode=BMM_Additive
    translationSpace=BCS_BoneSpace
    alphaBoolBlend {blendOption=Linear}
    boneToModify {boneName=base_r_hair}
   }
   n5 = "Transform (Modify) Bone - Bone: base_l_hair"(ComponentPose=@n4, Alpha=@n1, Translation="0,0,0", Rotation=Get RotationHair_Left() -> (RotationHair_Left:Rotator), Scale="1,1,1") {
    rotationMode=BMM_Additive
    rotationSpace=BCS_BoneSpace
    translationMode=BMM_Additive
    translationSpace=BCS_BoneSpace
    alphaBoolBlend {blendOption=Linear}
    boneToModify {boneName=base_l_hair}
   }
   n6 = RigidBody!(ComponentPose=@n5, Alpha, OverrideWorldGravity:Vector=Get GravityOverride(), ExternalForce) -> (Pose) {
    bClampLinearTranslationLimitToRefPose=true
    bOverrideWorldGravity=true
    overridePhysicsAsset=&/F_MED_KettlePress_FaceAcc_Physics
    simulationSpace=BaseBoneSpace
    alphaBoolBlend {blendOption=Linear}
    baseBoneRef {boneName=neck_02}
    componentAppliedLinearAccClamp=<3000 3000 3000>
    componentLinearAccScale=<0.5 0.5 0.5>
    componentLinearVelScale=<0.5 0.5 0.5>
    overrideWorldGravity {z=-980}
   }
   n7 = "Transform (Modify) Bone - Bone: base_l_hair"(ComponentPose=Local To Component(LocalPose=Use cached pose 'StartPose'()), Alpha=1, Translation="0,-3,0", Rotation="10,-25,0", Scale="1,1,1") {
    rotationMode=BMM_Additive
    rotationSpace=BCS_BoneSpace
    scaleMode=BMM_Additive
    scaleSpace=BCS_BoneSpace
    translationMode=BMM_Additive
    translationSpace=BCS_BoneSpace
    alphaBoolBlend {blendOption=Linear}
    boneToModify {boneName=base_l_hair}
    rotation {pitch=10, yaw=-25}
    translation {y=-3}
   }
   n8 = "Transform (Modify) Bone - Bone: base_r_hair"(
    ComponentPose=@n7
    Alpha=1
    Translation="-3,0,3"
    Rotation=Get RotationHair_R_SkyFloatingDive() -> (RotationHair_R_SkyFloatingDive:Rotator)
    Scale="1,1,1"
   ) {
    rotationMode=BMM_Additive
    rotationSpace=BCS_BoneSpace
    scaleMode=BMM_Additive
    scaleSpace=BCS_BoneSpace
    translationMode=BMM_Additive
    translationSpace=BCS_BoneSpace
    alphaBoolBlend {blendOption=Linear}
    boneToModify {boneName=base_r_hair}
    translation {x=-3, z=3}
   }
   n9 = "Transform (Modify) Bone - Bone: base_r_side"(ComponentPose=@n8, Alpha=1, Translation="0,0,0", Rotation="-25,0,25", Scale="1,1,1") {
    rotationMode=BMM_Additive
    rotationSpace=BCS_BoneSpace
    scaleMode=BMM_Additive
    scaleSpace=BCS_BoneSpace
    translationMode=BMM_Additive
    translationSpace=BCS_BoneSpace
    alphaBoolBlend {blendOption=Linear}
    boneToModify {boneName=base_r_side}
    rotation {pitch=-25, roll=25}
   }
   n10 = RigidBody(ComponentPose=@n9, OverrideWorldGravity=Get GravityOverride()) {
    bClampLinearTranslationLimitToRefPose=true
    bOverrideWorldGravity=true
    overridePhysicsAsset=&/F_MED_KettlePress_FaceAcc_Skydiving_Physics
    simulationSpace=BaseBoneSpace
    alphaBoolBlend {blendOption=Linear}
    baseBoneRef {boneName=neck_02}
    componentAppliedLinearAccClamp=<3000 3000 3000>
    componentLinearAccScale=<0.5 0.5 0.5>
    componentLinearVelScale=<0.5 0.5 0.5>
    overrideWorldGravity {z=-980}
   }
   n11 = Blend Poses by bool(
    BlendPose_0=Use cached pose 'frontEndPhysics_Cache'()
    BlendPose_1=Blend Poses by bool(BlendPose_0=Component To Local(ComponentPose=@n10), BlendPose_1=Component To Local(ComponentPose=@n6), bActiveValue=Get bIsSkyDiving()) {blendType=Linear}
    bActiveValue=Get bIsInFrontEnd()
   ) {blendType=Linear}
   n12 = Use cached pose 'constraints_Cache'()
   n13 = NOT Boolean(A=Get bRigidBodyEnabled())
   n14 = Anim Dynamics - Bone: dyn_l_earring(
    ComponentPose=Local To Component(LocalPose=Blend Poses by bool(BlendPose_0=@n12, BlendPose_1=@n11, bActiveValue=@n13) {blendType=Linear})
    Alpha=1
    ExternalForce="0,0,0"
   ) {
    angularDampingOverride=0.99
    angularSpringConstant=50
    bAngularSpring=true
    bOverrideAngularDamping=true
    bUseGravityOverride=true
    bUsePlanarLimit=false
    simulationSpace=World
    alphaBoolBlend {blendOption=Linear}
    boundBone {boneName=dyn_l_earring}
    componentLinearAccScale=<1 1 1>
    componentLinearVelScale=<1 1 1>
    gravityOverride {z=-980}
    physicsBodyDefinitions [{
      boundBone {boneName=dyn_l_earring}
      boxExtents=<15 15 15>
      constraintSetup {angularLimitsMax=<0 10 10>, angularLimitsMin=<0 -10 -10>, angularTarget=<0 0 0>}
      localJointOffset=<3 0 0>
     }]
    relativeSpaceBone {boneName=head}
   }
   n15 = Anim Dynamics - Bone: dyn_r_earring(ComponentPose=@n14, Alpha=1, ExternalForce="0,0,0") {
    angularDampingOverride=0.99
    angularSpringConstant=50
    bAngularSpring=true
    bOverrideAngularDamping=true
    bUseGravityOverride=true
    bUsePlanarLimit=false
    simulationSpace=World
    alphaBoolBlend {blendOption=Linear}
    boundBone {boneName=dyn_r_earring}
    componentLinearAccScale=<1 1 1>
    componentLinearVelScale=<1 1 1>
    gravityOverride {z=-980}
    physicsBodyDefinitions [{
      boundBone {boneName=dyn_r_earring}
      boxExtents=<10 10 10>
      constraintSetup {angularLimitsMax=<0 0 15>, angularLimitsMin=<0 0 -15>, angularTarget=<0 0 0>}
      localJointOffset=<2 0 0>
     }]
    relativeSpaceBone {boneName=head}
   }
   Output Pose(Result=Component To Local(ComponentPose=@n15)) {layerGroup=None}
   n16 = "Constraint - dyn_back_hair_2 (P)"(ComponentPose=Local To Component(LocalPose=Use cached pose 'StartPose'()), Alpha=0.2, ConstraintWeights_0=1) {
    alphaBoolBlend {blendOption=Linear}
    boneToModify {boneName=dyn_back_hair_2}
    constraintSetup [{targetBone {boneName=spine_05}}]
    constraintWeights [1]
   }
   n17 = "Constraint - dyn_l_hair_2 (PP)"!(ComponentPose=@n16, Alpha=0.2, ConstraintWeights_0=0.2, ConstraintWeights_1:float=0.8) -> (Pose) {
    alphaBoolBlend {blendOption=Linear}
    boneToModify {boneName=dyn_l_hair_2}
    constraintSetup [{targetBone {boneName=spine_05}}, {targetBone {boneName=clavicle_l}}]
    constraintWeights [1,1]
   }
   n18 = "Constraint - dyn_r_hair__2 (PP)"!(ComponentPose=@n17, Alpha=0.2, ConstraintWeights_0=0.8, ConstraintWeights_1:float=0.2) -> (Pose) {
    alphaBoolBlend {blendOption=Linear}
    boneToModify {boneName=dyn_r_hair__2}
    constraintSetup [{targetBone {boneName=head}}, {targetBone {boneName=clavicle_r}}]
    constraintWeights [1,1]
   }
   n19 = "Constraint - dyn_back_hair_3 (P)"(ComponentPose=@n18, Alpha=0.5, ConstraintWeights_0=1) {
    alphaBoolBlend {blendOption=Linear}
    boneToModify {boneName=dyn_back_hair_3}
    constraintSetup [{targetBone {boneName=spine_05}}]
    constraintWeights [1]
   }
   n20 = "Constraint - dyn_l_hair_3 (PP)"!(ComponentPose=@n19, Alpha=0.5, ConstraintWeights_0=0.2, ConstraintWeights_1:float=0.8) -> (Pose) {
    alphaBoolBlend {blendOption=Linear}
    boneToModify {boneName=dyn_l_hair_3}
    constraintSetup [{targetBone {boneName=spine_05}}, {targetBone {boneName=head}}]
    constraintWeights [1,1]
   }
   n21 = "Constraint - dyn_r_hair__3 (PP)"!(ComponentPose=@n20, Alpha=0.7, ConstraintWeights_0=0.8, ConstraintWeights_1:float=0.2) -> (Pose) {
    alphaBoolBlend {blendOption=Linear}
    boneToModify {boneName=dyn_r_hair__3}
    constraintSetup [{targetBone {boneName=head}}, {targetBone {boneName=clavicle_r}}]
    constraintWeights [1,1]
   }
   n22 = "Constraint - dyn_r_side_2 (P)"(ComponentPose=@n21, Alpha=0.2, ConstraintWeights_0=1) {
    alphaBoolBlend {blendOption=Linear}
    boneToModify {boneName=dyn_r_side_2}
    constraintSetup [{targetBone {boneName=spine_05}}]
    constraintWeights [1]
   }
   n23 = "Constraint - dyn_r_side_3 (P)"(ComponentPose=@n22, Alpha=0.5, ConstraintWeights_0=1) {
    alphaBoolBlend {blendOption=Linear}
    boneToModify {boneName=dyn_r_side_3}
    constraintSetup [{targetBone {boneName=spine_05}}]
    constraintWeights [1]
   }
   Save cached pose 'constraints_Cache'(Pose=Component To Local(ComponentPose=@n23)) {cacheName=constraints_Cache}
   n24 = "Transform (Modify) Bone - Bone: base_l_hair"(ComponentPose=Local To Component(LocalPose=Use cached pose 'StartPose'()), Alpha=1, Translation="0,1,0", Rotation="0,-10,0", Scale="1,1,1") {
    rotationMode=BMM_Additive
    rotationSpace=BCS_BoneSpace
    scaleMode=BMM_Additive
    scaleSpace=BCS_BoneSpace
    translationMode=BMM_Additive
    alphaBoolBlend {blendOption=Linear}
    boneToModify {boneName=base_l_hair}
    rotation {yaw=-10}
    translation {y=1}
   }
   n25 = RigidBody(ComponentPose=@n24, OverrideWorldGravity=Get GravityOverride()) {
    bClampLinearTranslationLimitToRefPose=true
    bOverrideWorldGravity=true
    overridePhysicsAsset=&/F_MED_KettlePress_FaceAcc_FrontEnd_Physics
    simulationSpace=BaseBoneSpace
    alphaBoolBlend {blendOption=Linear}
    baseBoneRef {boneName=head}
    componentAppliedLinearAccClamp=<3000 3000 3000>
    componentLinearAccScale=<0.25 0.35 0.35>
    componentLinearVelScale=<0.25 0.35 0.35>
    overrideWorldGravity {z=-980}
    simSpaceSettings {maxAngularAcceleration=1000, maxAngularVelocity=3, maxLinearAcceleration=250, maxLinearVelocity=500, velocityScaleZ=0.1, worldAlpha=0.25}
   }
   Save cached pose 'frontEndPhysics_Cache'(Pose=Component To Local(ComponentPose=@n25)) {cacheName=frontEndPhysics_Cache}
   n26 = Layered blend per bone(
    BasePose=Copy Pose From Mesh() {bCopyCurves=true, bUseAttachedParent=true}
    BlendPoses_0=Copy Pose From Mesh(SourceMeshComponent=Get Body_Mesh())
   ) {layerSetup [{branchFilters [{boneName=clavicle_l}, {boneName=clavicle_r}, {boneName=upperarm_r}, {boneName=upperarm_l}]}]}
   Save cached pose 'StartPose'(Pose=@n26) {cacheName=StartPose}
  ]}
 EventGraph {nodes [
   Event OnParachuteClosed()
   > Set RotationHair_R_SkyFloatingDive(RotationHair_R_SkyFloatingDive:Rotator="-20,0,0") -> (Output_Get:Rotator)
   Event OnCharacterCustomizationLoaded()
   > n1 = Cast To PlayerPawn_Athena(Object=TryGetPawnOwner())
   > Set Body_Mesh(Body_Mesh=GetSkeletalMeshForPartType(self=@n1))
   Event BlueprintUpdateAnimation()
   > n2 = Sequence!(execute) -> (then_0, then_1, then_2:Exec)
   > n3 = Gravity Override State Function DEPRECATED(CharacterPartAnimInstance=Self-Reference(), GameTimeInSeconds=GetGameTimeInSeconds(), GravityOverrideParams=Get Gravity Override Params())
   > Set GravityOverride(GravityOverride=@n3)
   n4 = BreakVector(InVec=Get PawnVelocity())
   n5 = "float < float"(A="Absolute (Float)"(A=@n4.Z), B=1300)
   n6 = "float < float"(A="Absolute (Float)"(A=@n4), B=1300)
   n7 = "float < float"(A="Absolute (Float)"(A=@n4.Y), B=1300)
   > Set high velocity turn off anim dyn(high velocity turn off anim dyn:double="To Float (Boolean)"(InBool=AND Boolean!(A=@n6, B=@n7, C:bool=@n5) -> (ReturnValue))) -> (Output_Get:double)
   > Set head up(head up:double=GetAlphaFromReferenceTransformDelta(SkeletalMeshComponent=Get Body_Mesh(), BoneName=spine_03, BaseName=head, InputRange=(X=0,Y=45), TwistAxis="0,0,1")) -> (Output_Get:double)
   Set turn off anim dyn when jumping(execute=@n2.then_1, turn off anim dyn when jumping:double="To Float (Boolean)"(InBool=Get bIsJumping())) -> (Output_Get:double)
   n8 = CheckForBackbling()
   Set isWearingBackbling(execute=@n2.then_2, isWearingBackbling=AND Boolean(A=@n8, B=NOT Boolean(A=@n8.bIsBackblingACape), C=NOT Boolean(A=@n8.bIsBackblingLong)))
   > Set Is Wearing Cape(Is Wearing Cape=OR Boolean(A=@n8.bIsBackblingACape, B=@n8.bIsBackblingLong))
   Event OnParachuteOpened()
   > Set RotationHair_R_SkyFloatingDive(RotationHair_R_SkyFloatingDive="0, 0, 0")
   Event OnBeginCrouching()
   > Set RotationHair_Back(RotationHair_Back:Rotator="50,0,0") -> (Output_Get:Rotator)
   > Set RotationHair_SideR(RotationHair_SideR:Rotator="40,0,0") -> (Output_Get:Rotator)
   > Set RotationHair_Right(RotationHair_Right:Rotator="30,30,30") -> (Output_Get:Rotator)
   > Set RotationHair_Left(RotationHair_Left:Rotator="25,-25,0") -> (Output_Get:Rotator)
   Event OnEndCrouching()
   > Set RotationHair_Back(RotationHair_Back="30,0,0")
   > Set RotationHair_SideR(RotationHair_SideR="45,0,0")
   > Set RotationHair_Right(RotationHair_Right="45,45,45")
   > Set RotationHair_Left(RotationHair_Left="0, 0, 0")
  ]}
]
variables [
 Body_Mesh {type=SkeletalMeshComponent}
 Gravity Override Params {type=GravityOverrideParamsStruct}
 GravityOverride {type=Vector}
 head up {type=number}
 high velocity turn off anim dyn {type=number}
 Is Wearing Cape {type=boolean}
 isWearingBackbling {type=boolean}
 RotationHair_Back {type=Rotator}
 RotationHair_Left {type=Rotator}
 RotationHair_R_SkyFloatingDive {type=Rotator}
 RotationHair_Right {type=Rotator}
 RotationHair_SideR {type=Rotator}
 SkelMesh {type=SkeletalMeshComponent}
 turn off anim dyn when jumping {type=number}
]
